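; Parent burst gameplay cue for chain whip melee hits, playing impact effects at the effect context's hit location.
; Water bodies get a depth-aware splash and fluid-sim impulse, while other hits get world FX; both add tagged sound and force feedback.
; A radial physics field spawns unless mobile, Switch or PC performance mode is detected.
; The burst decal fades on landscape or can be destroyed; survivors are scaled, with random rotation except on third impacts.
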(fn OnBurst(EitherVFX-preferthisversion)
  (params (MyTarget :type Actor%20Object%20Reference) (Parameters :type Gameplay%20Cue%20Parameters%20Structure%20%28by%20ref%29) (ParticleComponents :type Array%20of%20FXSystem%20Component%20Object%20References) (AudioComponents :type Array%20of%20Audio%20Component%20Object%20References) (BurstCameraShakeInstance :type Camera%20Shake%20Base%20Object%20Reference) (BurstDecalInstance :type Decal%20Actor%20Object%20Reference))
  (bind (_normalizedmagnitude _rawmagnitude _effectcontext) (Ability|GameplayCue|BreakGameplayCueParameters Parameters))
  (bind _returnvalue (Ability|EffectContext|GetHitResult _effectcontext))
  (bind (_bblockinghit _binitialoverlap _time _distance _location _impactpoint _normal _impactnormal _physmat _hitactor) (Collision|BreakHitResult _returnvalue))
  (bind (_normalizedmagnitude_2 _rawmagnitude_2 _effectcontext_2) (Ability|GameplayCue|BreakGameplayCueParameters Parameters))
  (bind (_returnvalue_bblockinghit _returnvalue_binitialoverlap _returnvalue_time _returnvalue_distance _returnvalue_location _returnvalue_impactpoint _returnvalue_normal _returnvalue_impactnormal _returnvalue_physmat _returnvalue_hitactor) (Collision|BreakHitResult (Ability|EffectContext|GetHitResult _effectcontext_2)))
  (bind (_aslandscape _bsuccess) (Utilities|Casting|CastToLandscape _returnvalue_hitactor))
  (bind _mytarget (Variables|GetMyTarget))
  (bind _mytarget_2 (Variables|GetMyTarget))
  (Utilities|FlowControl|Sequence
    (:then_0
      ;[0 1 0.2892] Calculate local variables that we will use later
      (LocalVariables|SetPreferPerformance (Math|Boolean|ORBoolean (Fortnite|IsMobilePlatform) (Class|PlatformHelpersLibrary|IsPcPerfMode MyTarget) (TouchPlatform|IsMobileGame MyTarget) (TouchPlatform|IsNintendoSwitchorNintendoSwitchPreview)))
      (LocalVariables|SetImpactLocation _impactpoint)
      (bind _mytarget_3 (Variables|GetMyTarget)
        (:then
          (LocalVariables|SetFX%20Rotation (Math|Conversions|RotationFromXVector (Math|Vector|Normalize (- (Transformation|GetActorLocation _mytarget_3) (LocalVariables|GetImpactLocation))))))
        (:else)))
    (:then_1
      (Utilities|IsValid MyTarget
        (:Is%20Valid
          (bind _asfort_water_body_bp (Utilities|Casting|CastToFortWaterBodyBP _hitactor)
            (:then
              ;[0.03323 0.1979 0.1038] Handle impact on water.
              (bind _outwatersurfaceinfo (WaterBody|GetWaterSurfaceInfoStruct _asfort_water_body_bp (LocalVariables|GetImpactLocation) true true false))
              (CallFunction|PlayForceFeedback _mytarget true)
              (bind _returnvalue_2 (Effects|Components|ParticleSystem|SpawnFXAtLocation _mytarget (Variables|ChainWhip|Water|GetWaterImpactFXSystem) (Utilities|Struct|BreakWaterSurfaceInfo _outwatersurfaceinfo) (LocalVariables|GetFX%20Rotation) "1.000000,1.000000,1.000000" true "AutoRelease")) ; Play splash fx
              (Utilities|IsValid _returnvalue_2
                (:Is%20Valid
                  (Transformation|AddRelativeRotation _returnvalue_2 (Variables|ChainWhip|World|GetWorldImpactFXSystemRotationOffset) false false)
                  (Effects|Components|ParticleSystem|SetFloatParameter _returnvalue_2 "WaterDepth" (.WaterDepth (Utilities|Struct|BreakWaterSurfaceInfo _outwatersurfaceinfo)))
                  (Effects|Components|ParticleSystem|SetFloatParameter _returnvalue_2 "WaterBodyIndex" (Math|Conversions|ToFloat(Integer) (.WaterBodyIdx (Utilities|Struct|BreakWaterSurfaceInfo _outwatersurfaceinfo)))))
                (:Is%20Not%20Valid
                  _))
              (CallFunction|PlaySound (Variables|GetMyTarget) (Variables|ChainWhip|Water|GetWaterImpactSoundEventTag) (Utilities|Struct|BreakWaterSurfaceInfo _outwatersurfaceinfo))
              ;;[1 0.2488 0.2289] Water plays fluid force impulse material animation based on the passed in values.
              (bind _fluidsimbp (Variables|/Game/Athena/Apollo/Environments/Landscape/Water/FortWaterBodyBP.FortWaterBodyBP_C|GetFluidSimBP _asfort_water_body_bp)
                (:then
                  (Class|BPFluidSim01|ApplyFluidForceImpulse _fluidsimbp (Utilities|Struct|MakeFluidForceImpulse (Utilities|Struct|BreakWaterSurfaceInfo _outwatersurfaceinfo) (Variables|ChainWhip|Water|GetWaterImpactForceRadius) (Variables|ChainWhip|Water|GetWaterImpactForceStrength) "/FortWater/FluidSimulation/Materials/Forces/M_Fluid_Sim_Force_Impulse.M_Fluid_Sim_Force_Impulse" 0.0 0.0 (Variables|ChainWhip|Water|GetWaterImpactWaterDropSplahes))))
                (:else)))
            (:CastFailed
              ;[0.01569 0.1623 0.2083] Handle impact against world.
              (bind _returnvalue_3 (Effects|Components|ParticleSystem|SpawnFXAtLocation _mytarget_2 (Variables|ChainWhip|World|GetWorldImpactFXSystem) (LocalVariables|GetImpactLocation) (LocalVariables|GetFX%20Rotation) "1.000000,1.000000,1.000000" true "AutoRelease"))
              (Utilities|IsValid _returnvalue_3
                (:Is%20Valid
                  (Transformation|AddRelativeRotation _returnvalue_3 (Variables|ChainWhip|World|GetWorldImpactFXSystemRotationOffset) false false))
                (:Is%20Not%20Valid
                  _))
              (CallFunction|PlaySound _mytarget_2 (Variables|ChainWhip|World|GetWorldImpactSoundEventTag) (LocalVariables|GetImpactLocation))
              (CallFunction|PlayForceFeedback _mytarget_2))))))
    (:then_2
      (if (not (LocalVariables|GetPreferPerformance))
        (Game|SpawnActorfromClass 0 "/Game/Effects/Niagara/Blueprints/PhsyicsFields/FN_RadialForce.FN_RadialForce_C" (Math|Conversions|ToTransform(Vector) _returnvalue_location) "Undefined" "MultiplyWithRoot" 0 3.0 0.75 4.8e+02 1.3 0.05 0)))
    (:then_3
      (if _bsuccess
        (Rendering|Components|Decal|SetFadeOut (Variables|/Script/Engine.DecalActor|GetDecal BurstDecalInstance) 0.0 3.0 true)
        (else
          ;[0.5477 0 1] Destroy the decal if 1) we are preferring performance or 2) it hits the train
          (if (not (LocalVariables|GetPreferPerformance))
            (Utilities|IsValid _returnvalue_hitactor
              (:Is%20Valid
                (bind _returnvalue_4 (GameplayTags|GetOwnedGameplayTags _returnvalue_hitactor :!class /Script/GameplayTags.GameplayTagAssetInterface))
                (if (Utilities|Array|ContainsItem (GameplayTags|BreakGameplayTagContainer _returnvalue_4) "(TagName=\"Athena.Gameplay.Train.Car\")")
                  _))))
          (Actor|DestroyActor (Variables|GetBurstDecalInstance))
          (LocalVariables|SetIsDecalBeingDestroyed true))))
    (:then_4
      ;[0.1795 0.2216 1] If the decal wasn't destroyed, set its size and rotation
      (if (not (LocalVariables|GetIsDecalBeingDestroyed))
        (Transformation|SetActorScale3D BurstDecalInstance (Variables|Default|GetDecal%20Size))
        (if (not (Variables|Default|GetIsThirdImpact))
          (Transformation|AddRelativeRotation (Variables|/Script/Engine.Actor|GetRootComponent BurstDecalInstance) (Math|Rotator|MakeRotator (.roll (LocalVariables|GetFX%20Rotation)) (+ (.pitch (LocalVariables|GetFX%20Rotation)) (Math|Float|Lerp -4e+01 4e+01 (Math|Random|RandomFloat))) (+ (.yaw (LocalVariables|GetFX%20Rotation)) (Math|Float|Lerp -4e+01 4e+01 (Math|Random|RandomFloat)))) false false))))
    (:then_5
      (return))))

(fn PlayForceFeedback
  (params (TargetActor :type Actor%20Object%20Reference) (bWaterImpact :type Boolean))
  (bind _returnvalue (Fortnite|GetFortPlayerControllerfromActor TargetActor))
  ;[0.008391 0.1308 0.1406] Play force feedback on the target actor (the player that is instigating the attack)
  (Utilities|IsValid _returnvalue
    (:Is%20Valid
      (Game|Feedback|LocalPlayForceFeedback _returnvalue (select bWaterImpact (Variables|ChainWhip|Water|GetWaterImpactForceFeedbackEffect) (Variables|ChainWhip|World|GetWorldImpactForceFeedbackEffect)) false "None"))))

(fn PlaySound
  (params (Target :type Actor%20Object%20Reference) (EventName :type Gameplay%20Tag%20Structure) (Location :type Vector))
  (bind _returnvalue (WorldSubsystems|GetSoundLibrarySubsystem Target))
  ;[0.007682 0.1219 0.1354] PlaySound: Play the desired event name sound at the provided location.
  (Utilities|IsValid Target
    (:Is%20Valid
      (Utilities|IsValid _returnvalue
        (:Is%20Valid
          (SoundLibrary|PlaySoundSimple _returnvalue Target (Utilities|Struct|MakeSimpleSoundSettings EventName false false true false 1.0 1.0 0.0 0 0 "None" 0 "KeepWorldPosition" Location) :!class /Script/SoundLibrary.SoundLibrarySubsystem))))))
